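# Cloth collision physics asset for the F_MED_GothDevil female medium player body.
# Capsules on the pelvis and spine_01 form the only listed collision shapes, covering the lower torso.
# Bodies for thighs, calves, feet and clothFX thigh twist bones are present but list no shapes.
# Two constraints link each thigh to the pelvis with all swing and twist axes free.
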PhysicsAsset /BRCosmetics/Characters/Player/Female/Medium/Bodies/F_MED_GothDevil/Meshes/F_MED_GothDevil_ClothColliders
bodies {
 calf_l []
 calf_r []
 clothFX_thigh_twist_01_l []
 clothFX_thigh_twist_01_r []
 foot_l []
 foot_r []
 pelvis [{shapeType=Capsule, length=7.689, radius=9.8596, center=<-2.8663 -1.1231 -0.0737>}]
 spine_01 [{shapeType=Capsule, length=6.8115, radius=7.4463, shapeName=spine_01_capsule, center=<-3.5267 0.0175 0>}]
 thigh_l []
 thigh_r []
}
constraints [
 {
  bone1Name=thigh_l
  bone2Name=pelvis
  swing1LimitDegrees=45
  swing1Motion=Free
  swing2LimitDegrees=45
  swing2Motion=Free
  twistLimitDegrees=45
  twistMotion=Free
 }
 {
  bone1Name=thigh_r
  bone2Name=pelvis
  swing1LimitDegrees=45
  swing1Motion=Free
  swing2LimitDegrees=45
  swing2Motion=Free
  twistLimitDegrees=45
  twistMotion=Free
 }
]
